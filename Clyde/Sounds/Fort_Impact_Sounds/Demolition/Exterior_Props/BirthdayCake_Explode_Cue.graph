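# Sound cue for a birthday cake explosion with crowd audio, filed under demolition impact sounds for exterior props.
# A single wave player feeds the output directly, with no randomization or mixing.
# The cue uses the large explosives sound class, boosts volume, and sets a very large max distance.
# Its own concurrency override uses the PreventNew rule, blocking new instances once the limit is reached.
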SoundCue /Clyde/Sounds/Fort_Impact_Sounds/Demolition/Exterior_Props/BirthdayCake_Explode_Cue
root=WavePlayer_2
nodes [{id=WavePlayer_2, properties {soundWaveAssetPtr=&/BirthdayCake_Explode_Crowd_01}}]
properties {
 bOverrideConcurrency=true
 maxDistance=2097152
 soundClassObject=/Game/Sounds/FortSoundClassesAndModes/Fort_Explosives_Large_PSM
 volumeMultiplier=1.6
 concurrencyOverrides {resolutionRule=PreventNew}
}
