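# Sound cue for birthday cake prop impacts, filed under demolition exterior props in the Fort impact sounds.
# A Random node picks one of three equally weighted impact waves and feeds a root Modulator left at its usual settings.
# The cue plays at reduced volume in the Fort_Impacts sound class and references a shared object-impact concurrency set, alongside stored stop-oldest override values.
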SoundCue /Clyde/Sounds/Fort_Impact_Sounds/Demolition/Exterior_Props/BirthdayCake_Impact_Cue
root=Modulator_8
nodes [
 {id=Modulator_8, children [Random_3]}
 {id=Random_3, children [WavePlayer_2,WavePlayer_5,WavePlayer_8], properties {weights [1,1,1]}}
 {id=WavePlayer_2, properties {soundWaveAssetPtr=&/BirthdayCake_Impact_01}}
 {id=WavePlayer_5, properties {soundWaveAssetPtr=&/BirthdayCake_Impact_02}}
 {id=WavePlayer_8, properties {soundWaveAssetPtr=&/BirthdayCake_Impact_03}}
]
properties {
 maxDistance=2097152
 soundClassObject=&Fort_Impacts
 volumeMultiplier=0.64
 concurrencyOverrides {maxCount=3, resolutionRule=StopOldest}
 concurrencySet [&CONC_Object_Impacts_FourMax]
}
