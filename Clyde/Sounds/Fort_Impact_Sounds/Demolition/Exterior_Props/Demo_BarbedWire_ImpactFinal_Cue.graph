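# Sound cue for the final impact of demolishing a barbed wire exterior prop.
# A root pitch modulator feeds a quality switch whose first two slots randomly pick between two impact waves.
# The last two slots each play an optimized wave through a second, wider pitch modulator, adding variation without extra samples.
# It routes to the Fort impacts sound class and limits simultaneous voices with a four-max object-impact concurrency.
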SoundCue /Clyde/Sounds/Fort_Impact_Sounds/Demolition/Exterior_Props/Demo_BarbedWire_ImpactFinal_Cue
root=Modulator_2
nodes [
 {id=Modulator_2, children [QualityLevel_0], properties {pitchMax=1.07, pitchMin=0.92}}
 {id=Modulator_3, children [WavePlayer_2], properties {pitchMax=1.15, pitchMin=0.88}}
 {id=Modulator_4, children [WavePlayer_10], properties {pitchMax=1.15, pitchMin=0.88}}
 {id=QualityLevel_0, children [Random_1,Random_1,Modulator_3,Modulator_4]}
 {id=Random_1, children [WavePlayer_5,WavePlayer_7], properties {weights [1,1]}}
 {id=WavePlayer_10, properties {soundWaveAssetPtr=&/BarbedWire_ImpactFinal_opt}}
 {id=WavePlayer_2, properties {soundWaveAssetPtr=&BarbedWire_ImpactFinal_opt}}
 {id=WavePlayer_5, properties {soundWaveAssetPtr=&/BarbedWire_ImpactFinal_01}}
 {id=WavePlayer_7, properties {soundWaveAssetPtr=&/BarbedWire_ImpactFinal_02}}
]
properties {
 maxDistance=2097152
 soundClassObject=&Fort_Impacts_PSM
 volumeMultiplier=0.544
 concurrencyOverrides {maxCount=4}
 concurrencySet [&CONC_Object_Impacts_FourMax]
}
