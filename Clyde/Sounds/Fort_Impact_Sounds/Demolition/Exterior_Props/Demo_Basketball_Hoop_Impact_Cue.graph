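# Sound cue for impact hits on a basketball hoop exterior prop in the demolition impact sound set, routed to the Fort_Impacts sound class.
# A random node picks one of two equally weighted hoop hit waves, then a modulator varies pitch slightly before an attenuation node at the root.
# The cue plays at reduced volume and uses an object-impact concurrency set that limits simultaneous instances.
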SoundCue /Clyde/Sounds/Fort_Impact_Sounds/Demolition/Exterior_Props/Demo_Basketball_Hoop_Impact_Cue
root=Attenuation_2
nodes [
 {id=Attenuation_2, children [Modulator_2]}
 {id=Modulator_2, children [Random_0], properties {pitchMax=1.07, pitchMin=0.92}}
 {id=Random_0, children [WavePlayer_2,WavePlayer_3], properties {weights [1,1]}}
 {id=WavePlayer_2, properties {soundWaveAssetPtr=/Clyde/Sounds/Fort_Impact_Sounds/Demolition/Basketball_Hoop_Hit02}}
 {id=WavePlayer_3, properties {soundWaveAssetPtr=/Clyde/Sounds/Fort_Impact_Sounds/Demolition/Basketball_Hoop_Hit01}}
]
properties {
 maxDistance=2097152
 soundClassObject=&Fort_Impacts
 volumeMultiplier=0.68
 concurrencyOverrides {maxCount=4}
 concurrencySet [&CONC_Object_Impacts_FourMax]
}
