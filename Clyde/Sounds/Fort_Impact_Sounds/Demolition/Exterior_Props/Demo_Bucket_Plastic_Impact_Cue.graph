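# Sound cue for plastic bucket impacts among exterior demolition props, routed to the Fort_Impacts sound class.
# A Random node picks one of three equally weighted bucket hit waves and feeds a Modulator that slightly varies pitch on each play.
# It plays at reduced volume and references a shared object-impact concurrency set that, by its name, caps simultaneous impacts at four.
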SoundCue /Clyde/Sounds/Fort_Impact_Sounds/Demolition/Exterior_Props/Demo_Bucket_Plastic_Impact_Cue
root=Modulator_6
nodes [
 {id=Modulator_6, children [Random_7], properties {pitchMax=1.07, pitchMin=0.92}}
 {id=Random_7, children [WavePlayer_38, WavePlayer_39, WavePlayer_40], properties {weights [1,1,1]}}
 {id=WavePlayer_38, properties {soundWaveAssetPtr=/Clyde/Sounds/Fort_Impact_Sounds/Demolition/Bucket_Plastic_Hit02}}
 {id=WavePlayer_39, properties {soundWaveAssetPtr=/Clyde/Sounds/Fort_Impact_Sounds/Demolition/Bucket_Plastic_Hit03}}
 {id=WavePlayer_40, properties {soundWaveAssetPtr=/Clyde/Sounds/Fort_Impact_Sounds/Demolition/Bucket_Plastic_Hit04}}
]
properties {
 maxDistance=2097152
 soundClassObject=&Fort_Impacts
 volumeMultiplier=0.68
 concurrencyOverrides {maxCount=4}
 concurrencySet [&CONC_Object_Impacts_FourMax]
}
