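# Sound cue for the final impact of a demolished car prop, layering a hard impact with an explosion.
# A mixer blends two branches, each randomly picking a wave (hard impacts or car explosions) and passing it through a modulator and its own attenuation override.
# The explosion branch sets a custom falloff and shape, while the impact branch keeps default attenuation values.
# The cue uses the impacts sound class and an object-impact concurrency set capped at four.
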SoundCue /Clyde/Sounds/Fort_Impact_Sounds/Demolition/Exterior_Props/Demo_Car_ImpactFinal_Cue
root=Mixer_3
nodes [
 {
  id=Attenuation_4
  children [Modulator_0]
  properties {bOverrideAttenuation=true, attenuationOverrides {falloffDistance=3000, attenuationShapeExtents {x=1000}}}
 }
 {id=Attenuation_5, children [Modulator_1], properties {bOverrideAttenuation=true}}
 {id=Mixer_3, children [Attenuation_5, Attenuation_4], properties {inputVolume [1,1]}}
 {id=Modulator_0, children [Random_5]}
 {id=Modulator_1, children [Random_4]}
 {id=Random_4, children [WavePlayer_12, WavePlayer_13, WavePlayer_14, WavePlayer_15], properties {weights [1,1,1,1]}}
 {id=Random_5, children [WavePlayer_18, WavePlayer_19, WavePlayer_20], properties {weights [1,1,1]}}
 {id=WavePlayer_12, properties {soundWaveAssetPtr=/Clyde/Sounds/Fort_Impact_Sounds/Demolition/Car_Impact_Hard1}}
 {id=WavePlayer_13, properties {soundWaveAssetPtr=/Clyde/Sounds/Fort_Impact_Sounds/Demolition/Car_Impact_Hard2}}
 {id=WavePlayer_14, properties {soundWaveAssetPtr=/Clyde/Sounds/Fort_Impact_Sounds/Demolition/Car_Impact_Hard3}}
 {id=WavePlayer_15, properties {soundWaveAssetPtr=/Clyde/Sounds/Fort_Impact_Sounds/Demolition/Car_Impact_Hard4}}
 {id=WavePlayer_18, properties {soundWaveAssetPtr=/Clyde/Sounds/Fort_Impact_Sounds/Demolition/Car_Explode1}}
 {id=WavePlayer_19, properties {soundWaveAssetPtr=/Clyde/Sounds/Fort_Impact_Sounds/Demolition/Car_Explode2}}
 {id=WavePlayer_20, properties {soundWaveAssetPtr=/Clyde/Sounds/Fort_Impact_Sounds/Demolition/Car_Explode3}}
]
properties {
 maxDistance=4000
 soundClassObject=&Fort_Impacts_PSM
 volumeMultiplier=0.544
 concurrencyOverrides {maxCount=4}
 concurrencySet [&CONC_Object_Impacts_FourMax]
}
