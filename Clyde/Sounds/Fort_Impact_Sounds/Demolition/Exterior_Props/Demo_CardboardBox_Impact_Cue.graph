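# Sound cue for cardboard box impacts, filed under demolition exterior props.
# A Mixer layers a random cardboard impact, given slight random pitch variation by a Modulator, over a quieter, unmodulated random wooden cabinet hit.
# Each layer picks from four equally weighted waves.
# The cue uses the Fort_Impacts sound class and a shared object-impact concurrency set named for a four-voice limit.
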SoundCue /Clyde/Sounds/Fort_Impact_Sounds/Demolition/Exterior_Props/Demo_CardboardBox_Impact_Cue
root=Mixer_0
nodes [
 {id=Mixer_0, children [Modulator_7,Random_0], properties {inputVolume [1.2,0.6]}}
 {id=Modulator_7, children [Random_1], properties {pitchMax=1.07, pitchMin=0.92}}
 {id=Random_0, children [WavePlayer_0,WavePlayer_1,WavePlayer_2,WavePlayer_3], properties {weights [1,1,1,1]}}
 {id=Random_1, children [WavePlayer_5,WavePlayer_6,WavePlayer_7,WavePlayer_8], properties {weights [1,1,1,1]}}
 {id=WavePlayer_0, properties {soundWaveAssetPtr=/Clyde/Sounds/Fort_Impact_Sounds/Demolition/Cabinet_Wood_Hit2}}
 {id=WavePlayer_1, properties {soundWaveAssetPtr=/Clyde/Sounds/Fort_Impact_Sounds/Demolition/Cabinet_Wood_Hit3}}
 {id=WavePlayer_2, properties {soundWaveAssetPtr=/Clyde/Sounds/Fort_Impact_Sounds/Demolition/Cabinet_Wood_Hit4}}
 {id=WavePlayer_3, properties {soundWaveAssetPtr=/Clyde/Sounds/Fort_Impact_Sounds/Demolition/Cabinet_Wood_Hit7}}
 {id=WavePlayer_5, properties {soundWaveAssetPtr=/Clyde/Sounds/Fort_Impact_Sounds/Demolition/cardboard_A_impact_01}}
 {id=WavePlayer_6, properties {soundWaveAssetPtr=/Clyde/Sounds/Fort_Impact_Sounds/Demolition/cardboard_A_impact_02}}
 {id=WavePlayer_7, properties {soundWaveAssetPtr=/Clyde/Sounds/Fort_Impact_Sounds/Demolition/cardboard_A_impact_03}}
 {id=WavePlayer_8, properties {soundWaveAssetPtr=/Clyde/Sounds/Fort_Impact_Sounds/Demolition/cardboard_A_impact_04}}
]
properties {
 maxDistance=2097152
 soundClassObject=&Fort_Impacts
 volumeMultiplier=0.544
 concurrencyOverrides {maxCount=4}
 concurrencySet [&CONC_Object_Impacts_FourMax]
}
