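# Sound cue for dumpster impacts, part of the exterior-prop demolition impact sounds.
# A Random node picks one of three equally weighted dumpster impact waves and feeds a Modulator that slightly varies pitch on each play.
# The cue plays at reduced volume in the Fort_Impacts sound class and uses the shared CONC_Object_Impacts_FourMax concurrency set to limit overlapping hits.
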SoundCue /Clyde/Sounds/Fort_Impact_Sounds/Demolition/Exterior_Props/Demo_Dumpster_Impact_Cue
root=Modulator_10
nodes [
 {id=Modulator_10, children [Random_22], properties {pitchMax=1.07, pitchMin=0.92}}
 {id=Random_22, children [WavePlayer_75, WavePlayer_76, WavePlayer_78], properties {weights [1,1,1]}}
 {id=WavePlayer_75, properties {soundWaveAssetPtr=/Clyde/Sounds/Fort_Impact_Sounds/Demolition/Dumpster_Impact01}}
 {id=WavePlayer_76, properties {soundWaveAssetPtr=/Clyde/Sounds/Fort_Impact_Sounds/Demolition/Dumpster_Impact02}}
 {id=WavePlayer_78, properties {soundWaveAssetPtr=/Clyde/Sounds/Fort_Impact_Sounds/Demolition/Dumpster_Impact04}}
]
properties {
 maxDistance=2097152
 soundClassObject=&Fort_Impacts
 volumeMultiplier=0.68
 concurrencyOverrides {resolutionRule=PreventNew}
 concurrencySet [&CONC_Object_Impacts_FourMax]
}
