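# Sound cue for impacts on small flower props in the exterior demolition set.
# A Random node picks one of four equally weighted flower demolition waves, which feeds a Modulator that randomizes pitch well above normal.
# It plays through the Fort_Impacts sound class at reduced volume, with object-impact concurrency limiting simultaneous plays to four.
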SoundCue /Clyde/Sounds/Fort_Impact_Sounds/Demolition/Exterior_Props/Demo_FlowersSmall_Impact_Cue
root=Modulator_0
nodes [
 {id=Modulator_0, children [Random_0], properties {pitchMax=1.5, pitchMin=1.4}}
 {id=Random_0, children [WavePlayer_0,WavePlayer_1,WavePlayer_2,WavePlayer_3], properties {weights [1,1,1,1]}}
 {id=WavePlayer_0, properties {soundWaveAssetPtr=&/Demo_Flowers_01}}
 {id=WavePlayer_1, properties {soundWaveAssetPtr=&/Demo_Flowers_02}}
 {id=WavePlayer_2, properties {soundWaveAssetPtr=&/Demo_Flowers_03}}
 {id=WavePlayer_3, properties {soundWaveAssetPtr=&/Demo_Flowers_04}}
]
properties {
 maxDistance=2097152
 soundClassObject=&Fort_Impacts
 volumeMultiplier=0.544
 concurrencyOverrides {maxCount=4, resolutionRule=StopOldest}
 concurrencySet [&CONC_Object_Impacts_FourMax]
}
