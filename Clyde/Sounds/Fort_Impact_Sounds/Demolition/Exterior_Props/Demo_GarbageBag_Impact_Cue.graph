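# Sound cue for garbage bag impacts, filed under demolition exterior props and assigned to the Fort_Impacts sound class.
# A quality-level root selects among four branches, stepping from three random wave variants through two branches of two down to a single wave.
# Each branch applies slight random pitch modulation to vary repeats.
# A shared object-impact concurrency set, named for a four-voice maximum, governs simultaneous playback.
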SoundCue /Clyde/Sounds/Fort_Impact_Sounds/Demolition/Exterior_Props/Demo_GarbageBag_Impact_Cue
root=QualityLevel_0
nodes [
 {id=Modulator_0, children [Random_0], properties {pitchMax=1.08, pitchMin=0.92}}
 {id=Modulator_1, children [Random_1], properties {pitchMax=1.08, pitchMin=0.92}}
 {id=Modulator_2, children [WavePlayer_5], properties {pitchMax=1.1, pitchMin=0.9}}
 {id=Modulator_6, children [Random_11], properties {pitchMax=1.08, pitchMin=0.92}}
 {id=QualityLevel_0, children [Modulator_6,Modulator_0,Modulator_1,Modulator_2]}
 {id=Random_0, children [WavePlayer_0,WavePlayer_1], properties {weights [1,1]}}
 {id=Random_1, children [WavePlayer_3,WavePlayer_4], properties {weights [1,1]}}
 {id=Random_11, children [WavePlayer_43, WavePlayer_46, WavePlayer_47], properties {weights [1,1,1]}}
 {id=WavePlayer_0, properties {soundWaveAssetPtr=/Clyde/Sounds/Fort_Impact_Sounds/Demolition/garbagebag_impact_01}}
 {id=WavePlayer_1, properties {soundWaveAssetPtr=/Clyde/Sounds/Fort_Impact_Sounds/Demolition/garbagebag_impact_04}}
 {id=WavePlayer_3, properties {soundWaveAssetPtr=&garbagebag_impact_01}}
 {id=WavePlayer_4, properties {soundWaveAssetPtr=&garbagebag_impact_04}}
 {id=WavePlayer_43, properties {soundWaveAssetPtr=&garbagebag_impact_01}}
 {id=WavePlayer_46, properties {soundWaveAssetPtr=&garbagebag_impact_04}}
 {id=WavePlayer_47, properties {soundWaveAssetPtr=/Clyde/Sounds/Fort_Impact_Sounds/Demolition/garbagebag_impact_05}}
 {id=WavePlayer_5, properties {soundWaveAssetPtr=&garbagebag_impact_01}}
]
properties {
 maxDistance=2097152
 soundClassObject=&Fort_Impacts
 volumeMultiplier=0.68
 concurrencyOverrides {resolutionRule=PreventNew}
 concurrencySet [&CONC_Object_Impacts_FourMax]
}
