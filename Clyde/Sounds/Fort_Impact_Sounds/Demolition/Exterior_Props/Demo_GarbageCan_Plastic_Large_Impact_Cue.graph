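# Sound cue for demolition impacts on a large plastic garbage can exterior prop.
# A random node picks one of three impact waves with equal weight, then a modulator applies slight pitch variation.
# It plays in the Fort_Impacts sound class at reduced volume and references a shared object-impact concurrency set, alongside an override rule set to prevent new sounds.
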SoundCue /Clyde/Sounds/Fort_Impact_Sounds/Demolition/Exterior_Props/Demo_GarbageCan_Plastic_Large_Impact_Cue
root=Modulator_10
nodes [
 {id=Modulator_10, children [Random_13], properties {pitchMax=1.07, pitchMin=0.92}}
 {id=Random_13, children [WavePlayer_169, WavePlayer_170, WavePlayer_171], properties {weights [1,1,1]}}
 {id=WavePlayer_169, properties {soundWaveAssetPtr=/Clyde/Sounds/Fort_Impact_Sounds/Demolition/GarbageCan_Plastic_Large_Impact03}}
 {id=WavePlayer_170, properties {soundWaveAssetPtr=/Clyde/Sounds/Fort_Impact_Sounds/Demolition/GarbageCan_Plastic_Large_Impact04}}
 {id=WavePlayer_171, properties {soundWaveAssetPtr=/Clyde/Sounds/Fort_Impact_Sounds/Demolition/GarbageCan_Plastic_Large_Impact05}}
]
properties {
 maxDistance=2097152
 soundClassObject=&Fort_Impacts
 volumeMultiplier=0.68
 concurrencyOverrides {resolutionRule=PreventNew}
 concurrencySet [&CONC_Object_Impacts_FourMax]
}
