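# Sound cue for the final impact of glass bottle props in demolition, under the exterior props set.
# A mixer layers two randomly chosen sounds: one of four glass bottle final-impact waves and one of three shelf bottle impact waves.
# Each layer has its own slight random pitch modulation for variation.
# It plays through the Fort impacts sound class and references a four-max object impacts concurrency set.
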SoundCue /Clyde/Sounds/Fort_Impact_Sounds/Demolition/Exterior_Props/Demo_Glass_Bottles_ImpactFinal_Cue
root=Mixer_3
nodes [
 {id=Mixer_3, children [Modulator_2,Modulator_5], properties {inputVolume [1,1]}}
 {id=Modulator_2, children [Random_3], properties {pitchMax=1.07, pitchMin=0.92}}
 {id=Modulator_5, children [Random_7], properties {pitchMax=1.07, pitchMin=0.92}}
 {id=Random_3, children [WavePlayer_12, WavePlayer_13, WavePlayer_14, WavePlayer_15], properties {weights [1,1,1,1]}}
 {id=Random_7, children [WavePlayer_25, WavePlayer_26, WavePlayer_27], properties {weights [1,1,1]}}
 {id=WavePlayer_12, properties {soundWaveAssetPtr=/Clyde/Sounds/Fort_Impact_Sounds/Demolition/Glass_Bottle_ImpactFinal01}}
 {id=WavePlayer_13, properties {soundWaveAssetPtr=/Clyde/Sounds/Fort_Impact_Sounds/Demolition/Glass_Bottle_ImpactFinal02}}
 {id=WavePlayer_14, properties {soundWaveAssetPtr=/Clyde/Sounds/Fort_Impact_Sounds/Demolition/Glass_Bottle_ImpactFinal03}}
 {id=WavePlayer_15, properties {soundWaveAssetPtr=/Clyde/Sounds/Fort_Impact_Sounds/Demolition/Glass_Bottle_ImpactFinal04}}
 {id=WavePlayer_25, properties {soundWaveAssetPtr=/Clyde/Sounds/Fort_Impact_Sounds/Demolition/Shelf_Bottle_Impact02}}
 {id=WavePlayer_26, properties {soundWaveAssetPtr=/Clyde/Sounds/Fort_Impact_Sounds/Demolition/Shelf_Bottle_Impact05}}
 {id=WavePlayer_27, properties {soundWaveAssetPtr=/Clyde/Sounds/Fort_Impact_Sounds/Demolition/Shelf_Bottle_Impact06}}
]
properties {
 maxDistance=2097152
 soundClassObject=&Fort_Impacts_PSM
 volumeMultiplier=0.544
 concurrencyOverrides {resolutionRule=PreventNew}
 concurrencySet [&CONC_Object_Impacts_FourMax]
}
