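# Sound cue for the final demolition impact on an exterior fire hydrant prop, routed to the Fort impacts sound class.
# A mixer layers a lightly pitch-varied hydrant hit with a random metal cannon impact and a random metal railing death.
# The cannon and railing layers are pitched up sharply, repurposing those samples, and the mixer's fourth input is empty.
# An object-impact concurrency set caps simultaneous voices at four.
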SoundCue /Clyde/Sounds/Fort_Impact_Sounds/Demolition/Exterior_Props/Demo_Hydrant_ImpactFinal_Cue
root=Mixer_11
nodes [
 {id=Mixer_11, children [Modulator_14,Modulator_2,Modulator_15,null], properties {inputVolume [0.8,0.75,1,1]}}
 {id=Modulator_14, children [WavePlayer_53], properties {pitchMax=1.07, pitchMin=0.92}}
 {id=Modulator_15, children [Random_13], properties {pitchMax=2, pitchMin=1.75}}
 {id=Modulator_2, children [Random_2], properties {pitchMax=2, pitchMin=1.8}}
 {id=Random_13, children [WavePlayer_52, WavePlayer_54], properties {weights [1,1]}}
 {id=Random_2, children [WavePlayer_37, WavePlayer_38], properties {weights [1,1]}}
 {id=WavePlayer_37, properties {soundWaveAssetPtr=/Clyde/Sounds/Fort_Impact_Sounds/Demolition/Metal_Cannon_ImpactFinal01}}
 {id=WavePlayer_38, properties {soundWaveAssetPtr=/Clyde/Sounds/Fort_Impact_Sounds/Demolition/Metal_Cannon_ImpactFinal02}}
 {id=WavePlayer_52, properties {soundWaveAssetPtr=/Clyde/Sounds/Fort_Impact_Sounds/Demolition/Metal_Railing_Death04}}
 {id=WavePlayer_53, properties {soundWaveAssetPtr=&/metal_hydrant_hit_01}}
 {id=WavePlayer_54, properties {soundWaveAssetPtr=/Clyde/Sounds/Fort_Impact_Sounds/Demolition/Metal_Railing_Death03}}
]
properties {
 maxDistance=2097152
 soundClassObject=&Fort_Impacts_PSM
 volumeMultiplier=0.408
 concurrencyOverrides {maxCount=4}
 concurrencySet [&CONC_Object_Impacts_FourMax]
}
