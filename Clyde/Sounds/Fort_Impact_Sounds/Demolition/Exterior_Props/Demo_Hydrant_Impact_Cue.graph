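# Sound cue for a fire hydrant impact in the exterior-prop demolition set, routed to the Fort_Impacts class.
# A quality switch steps from a three-layer mix to a two-layer mix, then to a single optimized wave for the last two levels.
# The mixes combine a quieter hydrant hit with random small metal impacts; the fullest adds a higher-pitched pole impact that can randomly be silent.
# Every branch has random pitch variation, and concurrency caps simultaneous impacts at four.
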SoundCue /Clyde/Sounds/Fort_Impact_Sounds/Demolition/Exterior_Props/Demo_Hydrant_Impact_Cue
root=QualityLevel_0
nodes [
 {id=Mixer_0, children [Modulator_0,Modulator_3], properties {inputVolume [0.25,1]}}
 {id=Mixer_11, children [Modulator_14,Modulator_8,Modulator_9], properties {inputVolume [0.25,1,1]}}
 {id=Modulator_0, children [WavePlayer_0], properties {pitchMax=1.07, pitchMin=0.92}}
 {id=Modulator_14, children [WavePlayer_53], properties {pitchMax=1.07, pitchMin=0.92}}
 {id=Modulator_3, children [Random_1], properties {pitchMax=1.07, pitchMin=0.92}}
 {id=Modulator_5, children [WavePlayer_8], properties {pitchMax=1.07, pitchMin=0.92}}
 {id=Modulator_6, children [WavePlayer_9], properties {pitchMax=1.07, pitchMin=0.92}}
 {id=Modulator_8, children [Random_7], properties {pitchMax=1.07, pitchMin=0.92}}
 {id=Modulator_9, children [Random_8], properties {pitchMax=1.3, pitchMin=1.15}}
 {id=QualityLevel_0, children [Mixer_11,Mixer_0,Modulator_5,Modulator_6]}
 {id=Random_1, children [WavePlayer_4,WavePlayer_5], properties {weights [1,1]}}
 {id=Random_7, children [WavePlayer_25, WavePlayer_26], properties {weights [1,1]}}
 {id=Random_8, children [WavePlayer_28, WavePlayer_29, null], properties {weights [1,1,1]}}
 {id=WavePlayer_0, properties {soundWaveAssetPtr=&/metal_hydrant_hit_01}}
 {id=WavePlayer_25, properties {soundWaveAssetPtr=/Clyde/Sounds/Fort_Impact_Sounds/Demolition/Metal_Small_Impact06}}
 {id=WavePlayer_26, properties {soundWaveAssetPtr=/Clyde/Sounds/Fort_Impact_Sounds/Demolition/Metal_Small_Impact04}}
 {id=WavePlayer_28, properties {soundWaveAssetPtr=/Clyde/Sounds/Fort_Impact_Sounds/Demolition/Metal_Pole_Impact01}}
 {id=WavePlayer_29, properties {soundWaveAssetPtr=/Clyde/Sounds/Fort_Impact_Sounds/Demolition/Metal_Pole_Impact05}}
 {id=WavePlayer_4, properties {soundWaveAssetPtr=&Metal_Small_Impact06}}
 {id=WavePlayer_5, properties {soundWaveAssetPtr=&Metal_Small_Impact04}}
 {id=WavePlayer_53, properties {soundWaveAssetPtr=&metal_hydrant_hit_01}}
 {id=WavePlayer_8, properties {soundWaveAssetPtr=&/Hydrant_Impact_opt}}
 {id=WavePlayer_9, properties {soundWaveAssetPtr=&Hydrant_Impact_opt}}
]
properties {
 maxDistance=2097152
 soundClassObject=&Fort_Impacts
 volumeMultiplier=0.408
 concurrencyOverrides {maxCount=4}
 concurrencySet [&CONC_Object_Impacts_FourMax]
}
