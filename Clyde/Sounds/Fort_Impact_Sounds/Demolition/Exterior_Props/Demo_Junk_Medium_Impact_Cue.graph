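# Sound cue for medium junk impacts, filed under fort demolition exterior prop sounds.
# A random node picks one of four equally weighted junk impact waves and feeds a modulator that slightly varies pitch on each play.
# The cue plays at reduced volume in the fort impacts sound class, with a four-max object impact concurrency set to limit simultaneous instances.
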SoundCue /Clyde/Sounds/Fort_Impact_Sounds/Demolition/Exterior_Props/Demo_Junk_Medium_Impact_Cue
root=Modulator_2
nodes [
 {id=Modulator_2, children [Random_17], properties {pitchMax=1.07, pitchMin=0.92}}
 {id=Random_17, children [WavePlayer_100, WavePlayer_101, WavePlayer_102, WavePlayer_103], properties {weights [1,1,1,1]}}
 {id=WavePlayer_100, properties {soundWaveAssetPtr=/Clyde/Sounds/Fort_Impact_Sounds/Demolition/Junk_Impact_Medium04}}
 {id=WavePlayer_101, properties {soundWaveAssetPtr=/Clyde/Sounds/Fort_Impact_Sounds/Demolition/Junk_Impact_Medium01}}
 {id=WavePlayer_102, properties {soundWaveAssetPtr=/Clyde/Sounds/Fort_Impact_Sounds/Demolition/Junk_Impact_Medium02}}
 {id=WavePlayer_103, properties {soundWaveAssetPtr=/Clyde/Sounds/Fort_Impact_Sounds/Demolition/Junk_Impact_Medium03}}
]
properties {
 maxDistance=2097152
 soundClassObject=&Fort_Impacts
 volumeMultiplier=0.68
 concurrencyOverrides {maxCount=4}
 concurrencySet [&CONC_Object_Impacts_FourMax]
}
