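# Sound cue for metal barrel impacts, filed under exterior demolition props and routed to the Fort_Impacts sound class.
# A Random node picks among three equally weighted barrel impact waves and feeds a Modulator at default settings for pitch and volume variation.
# The cue plays at reduced volume and references the CONC_Object_Impacts_FourMax concurrency set to limit overlapping impacts.
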SoundCue /Clyde/Sounds/Fort_Impact_Sounds/Demolition/Exterior_Props/Demo_MetalBarrel_Impact_Cue
root=Modulator_0
nodes [
 {id=Modulator_0, children [Random_1]}
 {id=Random_1, children [WavePlayer_3,WavePlayer_4,WavePlayer_5], properties {weights [1,1,1]}}
 {id=WavePlayer_3, properties {soundWaveAssetPtr=/Clyde/Sounds/Fort_Impact_Sounds/Demolition/Metal_Barrel_Impact04}}
 {id=WavePlayer_4, properties {soundWaveAssetPtr=/Clyde/Sounds/Fort_Impact_Sounds/Demolition/Metal_Barrel_Impact05}}
 {id=WavePlayer_5, properties {soundWaveAssetPtr=/Clyde/Sounds/Fort_Impact_Sounds/Demolition/Metal_Barrel_Impact06}}
]
properties {
 maxDistance=2097152
 soundClassObject=&Fort_Impacts
 volumeMultiplier=0.34
 concurrencyOverrides {maxCount=3, resolutionRule=StopOldest}
 concurrencySet [&CONC_Object_Impacts_FourMax]
}
