SoundCue /Clyde/Sounds/Fort_Impact_Sounds/Demolition/Exterior_Props/Demo_MetalElecTower_Impact_Cue
root=QualityLevel_0
nodes [
 {id=Mixer_0, children [Modulator_5,Modulator_4], properties {inputVolume [1,1]}}
 {id=Mixer_1, children [Modulator_1,Modulator_0], properties {inputVolume [1,1]}}
 {id=Modulator_0, children [Random_2], properties {pitchMax=0.9, pitchMin=0.75, volumeMax=1.1, volumeMin=0.8}}
 {id=Modulator_1, children [Random_0], properties {pitchMax=0.9, pitchMin=0.8, volumeMin=0.8}}
 {id=Modulator_3, children [WavePlayer_1]}
 {id=Modulator_4, children [Random_23], properties {pitchMax=0.9, pitchMin=0.75, volumeMax=1.1, volumeMin=0.8}}
 {id=Modulator_5, children [Random_17], properties {pitchMax=0.9, pitchMin=0.8, volumeMin=0.8}}
 {id=Modulator_7, children [WavePlayer_9]}
 {id=QualityLevel_0, children [Mixer_0,Mixer_1,Modulator_3,Modulator_7]}
 {id=Random_0, children [WavePlayer_2,WavePlayer_3], properties {weights [1,1]}}
 {id=Random_17, children [WavePlayer_60, WavePlayer_61, WavePlayer_62], properties {weights [1,1,1]}}
 {id=Random_2, children [WavePlayer_5,WavePlayer_6], properties {bRandomizeWithoutReplacement=false, weights [1,1]}}
 {id=Random_23, children [WavePlayer_38, WavePlayer_39], properties {bRandomizeWithoutReplacement=false, weights [1,1]}}
 {id=WavePlayer_1, properties {soundWaveAssetPtr=&/MetalElecTower_Impact}}
 {id=WavePlayer_2, properties {soundWaveAssetPtr=/Clyde/Sounds/Fort_Impact_Sounds/Demolition/Metal_Railing_Impact04}}
 {id=WavePlayer_3, properties {soundWaveAssetPtr=/Clyde/Sounds/Fort_Impact_Sounds/Demolition/Metal_Railing_Impact01}}
 {id=WavePlayer_38, properties {soundWaveAssetPtr=/Clyde/Sounds/Fort_Impact_Sounds/Demolition/Metal_Pole_Impact01}}
 {id=WavePlayer_39, properties {soundWaveAssetPtr=/Clyde/Sounds/Fort_Impact_Sounds/Demolition/Metal_Pole_Impact06}}
 {id=WavePlayer_5, properties {soundWaveAssetPtr=&Metal_Pole_Impact01}}
 {id=WavePlayer_6, properties {soundWaveAssetPtr=&Metal_Pole_Impact06}}
 {id=WavePlayer_60, properties {soundWaveAssetPtr=&Metal_Railing_Impact04}}
 {id=WavePlayer_61, properties {soundWaveAssetPtr=&Metal_Railing_Impact01}}
 {id=WavePlayer_62, properties {soundWaveAssetPtr=/Clyde/Sounds/Fort_Impact_Sounds/Demolition/Metal_Railing_Impact06}}
 {id=WavePlayer_9, properties {soundWaveAssetPtr=&MetalElecTower_Impact}}
]
properties {
 maxDistance=2097152
 soundClassObject=&Fort_Impacts
 volumeMultiplier=0.34
 concurrencyOverrides {maxCount=4, resolutionRule=StopOldest}
 concurrencySet [&CONC_Object_Impacts_FourMax]
}
