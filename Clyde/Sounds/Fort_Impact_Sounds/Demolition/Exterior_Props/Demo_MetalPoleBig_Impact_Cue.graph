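# Sound cue for a large metal pole impact on demolition exterior props, routed to the Fort_Impacts class.
# A quality node selects four progressively simpler branches, from random pole hits layered with random cannon impacts down to one pole wave.
# Pitch modulators vary every branch, with the cannon layer pitched lower and mixed quieter.
# The cue uses the CONC_Object_Impacts_FourMax concurrency set.
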SoundCue /Clyde/Sounds/Fort_Impact_Sounds/Demolition/Exterior_Props/Demo_MetalPoleBig_Impact_Cue
root=QualityLevel_0
nodes [
 {id=Mixer_0, children [Random_0,Modulator_1], properties {inputVolume [1,0.7]}}
 {id=Mixer_2, children [Random_4,Modulator_12], properties {inputVolume [1,0.7]}}
 {id=Modulator_0, children [Mixer_0], properties {pitchMax=1.07, pitchMin=0.92}}
 {id=Modulator_1, children [WavePlayer_3], properties {pitchMax=0.95, pitchMin=0.85}}
 {id=Modulator_10, children [Mixer_2], properties {pitchMax=1.07, pitchMin=0.92}}
 {id=Modulator_12, children [Random_24], properties {pitchMax=0.95, pitchMin=0.85}}
 {id=Modulator_2, children [Random_2], properties {pitchMax=1.07, pitchMin=0.92}}
 {id=Modulator_4, children [WavePlayer_7], properties {pitchMax=1.1, pitchMin=0.9}}
 {id=QualityLevel_0, children [Modulator_10,Modulator_0,Modulator_2,Modulator_4]}
 {id=Random_0, children [WavePlayer_0,WavePlayer_1], properties {weights [1,1]}}
 {id=Random_2, children [WavePlayer_4,WavePlayer_5], properties {weights [1,1]}}
 {id=Random_24, children [WavePlayer_87, WavePlayer_84], properties {weights [1,1]}}
 {id=Random_4, children [WavePlayer_36, WavePlayer_37], properties {weights [1,1]}}
 {id=WavePlayer_0, properties {soundWaveAssetPtr=/Clyde/Sounds/Fort_Impact_Sounds/Demolition/Metal_Pole_Impact01}}
 {id=WavePlayer_1, properties {soundWaveAssetPtr=/Clyde/Sounds/Fort_Impact_Sounds/Demolition/Metal_Pole_Impact06}}
 {id=WavePlayer_3, properties {soundWaveAssetPtr=/Clyde/Sounds/Fort_Impact_Sounds/Demolition/Metal_Cannon_Impact03}}
 {id=WavePlayer_36, properties {soundWaveAssetPtr=&Metal_Pole_Impact01}}
 {id=WavePlayer_37, properties {soundWaveAssetPtr=&Metal_Pole_Impact06}}
 {id=WavePlayer_4, properties {soundWaveAssetPtr=&Metal_Pole_Impact01}}
 {id=WavePlayer_5, properties {soundWaveAssetPtr=&Metal_Pole_Impact06}}
 {id=WavePlayer_7, properties {soundWaveAssetPtr=&Metal_Pole_Impact01}}
 {id=WavePlayer_84, properties {soundWaveAssetPtr=/Clyde/Sounds/Fort_Impact_Sounds/Demolition/Metal_Cannon_Impact02}}
 {id=WavePlayer_87, properties {soundWaveAssetPtr=&Metal_Cannon_Impact03}}
]
properties {
 maxDistance=2097152
 pitchMultiplier=0.9
 soundClassObject=&Fort_Impacts
 volumeMultiplier=0.34
 concurrencyOverrides {maxCount=3, resolutionRule=StopOldest}
 concurrencySet [&CONC_Object_Impacts_FourMax]
}
