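# Sound cue for the final demolition impact of an exterior metal railing prop.
# A quality-level root picks among four branches: the first two mix pitch-modulated random railing death sounds with random rock impacts, using three then two variants per layer.
# The last two branches each play a single pre-optimized wave with pitch variation to save cost.
# It uses the Fort impacts sound class and a four-max object-impact concurrency set.
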SoundCue /Clyde/Sounds/Fort_Impact_Sounds/Demolition/Exterior_Props/Demo_MetalRailing_ImpactFinal_Cue
root=QualityLevel_0
nodes [
 {id=Mixer_0, children [Modulator_0,Modulator_1], properties {inputVolume [1,1]}}
 {id=Mixer_5, children [Modulator_12,Modulator_13], properties {inputVolume [1,1]}}
 {id=Modulator_0, children [Random_0], properties {pitchMax=1.07, pitchMin=0.92}}
 {id=Modulator_1, children [Random_1]}
 {id=Modulator_12, children [Random_11], properties {pitchMax=1.07, pitchMin=0.92}}
 {id=Modulator_13, children [Random_12]}
 {id=Modulator_2, children [WavePlayer_0], properties {pitchMax=1.07, pitchMin=0.92}}
 {id=Modulator_3, children [WavePlayer_7], properties {pitchMax=1.07, pitchMin=0.92}}
 {id=QualityLevel_0, children [Mixer_5,Mixer_0,Modulator_2,Modulator_3]}
 {id=Random_0, children [WavePlayer_1,WavePlayer_2], properties {weights [1,1]}}
 {id=Random_1, children [WavePlayer_4,WavePlayer_5], properties {weights [1,1]}}
 {id=Random_11, children [WavePlayer_195, WavePlayer_196, WavePlayer_47], properties {weights [1,1,1]}}
 {id=Random_12, children [WavePlayer_48, WavePlayer_49, WavePlayer_50], properties {weights [1,1,1]}}
 {id=WavePlayer_0, properties {soundWaveAssetPtr=&/MetalRailing_ImpactFinal_opt}}
 {id=WavePlayer_1, properties {soundWaveAssetPtr=/Clyde/Sounds/Fort_Impact_Sounds/Demolition/Metal_Railing_Death02}}
 {id=WavePlayer_195, properties {soundWaveAssetPtr=&Metal_Railing_Death02}}
 {id=WavePlayer_196, properties {soundWaveAssetPtr=/Clyde/Sounds/Fort_Impact_Sounds/Demolition/Metal_Railing_Death04}}
 {id=WavePlayer_2, properties {soundWaveAssetPtr=&Metal_Railing_Death04}}
 {id=WavePlayer_4, properties {soundWaveAssetPtr=/Clyde/Sounds/Fort_Impact_Sounds/Demolition/rock_impact_02}}
 {id=WavePlayer_47, properties {soundWaveAssetPtr=/Clyde/Sounds/Fort_Impact_Sounds/Demolition/Metal_Railing_Death03}}
 {id=WavePlayer_48, properties {soundWaveAssetPtr=&rock_impact_02}}
 {id=WavePlayer_49, properties {soundWaveAssetPtr=/Clyde/Sounds/Fort_Impact_Sounds/Demolition/rock_impact_03}}
 {id=WavePlayer_5, properties {soundWaveAssetPtr=&rock_impact_03}}
 {id=WavePlayer_50, properties {soundWaveAssetPtr=/Clyde/Sounds/Fort_Impact_Sounds/Demolition/rock_impact_06}}
 {id=WavePlayer_7, properties {soundWaveAssetPtr=&MetalRailing_ImpactFinal_opt}}
]
properties {
 maxDistance=2097152
 soundClassObject=&Fort_Impacts_PSM
 volumeMultiplier=0.544
 concurrencyOverrides {resolutionRule=PreventNew}
 concurrencySet [&CONC_Object_Impacts_FourMax]
}
