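# Sound cue for a small metal trashcan impact, routed to the Fort_Impacts sound class.
# It mixes a random trashcan impact, pitched well up, with a quieter random wobble layer, each with three variants and its own pitch modulation.
# An attenuation node overrides attenuation, and playback uses a concurrency set for object impacts.
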SoundCue /Clyde/Sounds/Fort_Impact_Sounds/Demolition/Exterior_Props/Demo_MetalTrashcanSmall_Impact_Cue
root=Attenuation_10
nodes [
 {id=Attenuation_10, children [Mixer_7], properties {bOverrideAttenuation=true}}
 {id=Mixer_7, children [Modulator_7,Modulator_1], properties {inputVolume [1,0.3]}}
 {id=Modulator_1, children [Random_2], properties {pitchMax=1.2}}
 {id=Modulator_7, children [Random_19], properties {pitchMax=1.7, pitchMin=1.5}}
 {id=Random_19, children [WavePlayer_69, WavePlayer_70, WavePlayer_71], properties {weights [1,1,1]}}
 {id=Random_2, children [WavePlayer_10, WavePlayer_11, WavePlayer_12], properties {weights [1,1,1]}}
 {id=WavePlayer_10, properties {soundWaveAssetPtr=/Clyde/Sounds/Fort_Impact_Sounds/Demolition/Fort_Impact_Wobble_01}}
 {id=WavePlayer_11, properties {soundWaveAssetPtr=/Clyde/Sounds/Fort_Impact_Sounds/Demolition/Fort_Impact_Wobble_02}}
 {id=WavePlayer_12, properties {soundWaveAssetPtr=/Clyde/Sounds/Fort_Impact_Sounds/Demolition/Fort_Impact_Wobble_03}}
 {id=WavePlayer_69, properties {soundWaveAssetPtr=/Clyde/Sounds/Fort_Impact_Sounds/Demolition/metaltrashcan_impact_01}}
 {id=WavePlayer_70, properties {soundWaveAssetPtr=/Clyde/Sounds/Fort_Impact_Sounds/Demolition/metaltrashcan_impact_02}}
 {id=WavePlayer_71, properties {soundWaveAssetPtr=/Clyde/Sounds/Fort_Impact_Sounds/Demolition/metaltrashcan_impact_03}}
]
properties {
 maxDistance=4000
 soundClassObject=&Fort_Impacts
 volumeMultiplier=0.34
 concurrencyOverrides {resolutionRule=PreventNew}
 concurrencySet [&CONC_Object_Impacts_FourMax]
}
