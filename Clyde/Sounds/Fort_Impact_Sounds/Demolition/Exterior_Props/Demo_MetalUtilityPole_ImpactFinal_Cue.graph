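# Sound cue for the final impact of a demolished metal utility pole, routed to the impacts sound class with a four-max concurrency set.
# A quality-level node selects between two layered mixes and two single-wave fallbacks using an optimized asset.
# Each mix combines randomized big-metal, quieter pole-impact and pole-death layers, with more variants in the first.
# Modulators add pitch variation to many layers, plus volume variation on the fallbacks.
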SoundCue /Clyde/Sounds/Fort_Impact_Sounds/Demolition/Exterior_Props/Demo_MetalUtilityPole_ImpactFinal_Cue
root=QualityLevel_0
nodes [
 {id=Mixer_1, children [Modulator_7,Modulator_5,Modulator_6], properties {inputVolume [2,0.5,1]}}
 {id=Mixer_4, children [Random_6,Modulator_3,Modulator_8], properties {inputVolume [2,0.5,1]}}
 {id=Modulator_0, children [WavePlayer_3], properties {pitchMax=1.1, pitchMin=0.9, volumeMax=1.1, volumeMin=0.9}}
 {id=Modulator_1, children [WavePlayer_4], properties {pitchMax=1.1, pitchMin=0.9, volumeMax=1.1, volumeMin=0.9}}
 {id=Modulator_3, children [Random_5]}
 {id=Modulator_5, children [Random_3], properties {pitchMax=1.1, pitchMin=0.9}}
 {id=Modulator_6, children [Random_7], properties {pitchMax=1.1, pitchMin=0.9}}
 {id=Modulator_7, children [Random_4], properties {pitchMax=1.1, pitchMin=0.9}}
 {id=Modulator_8, children [Random_11], properties {pitchMax=1.1, pitchMin=0.9}}
 {id=QualityLevel_0, children [Mixer_4,Mixer_1,Modulator_0,Modulator_1]}
 {id=Random_11, children [WavePlayer_40, WavePlayer_38, WavePlayer_39], properties {weights [1,1,1]}}
 {id=Random_3, children [WavePlayer_13, WavePlayer_14], properties {bRandomizeWithoutReplacement=false, weights [1,1]}}
 {id=Random_4, children [WavePlayer_15, WavePlayer_16], properties {weights [1,1]}}
 {id=Random_5, children [WavePlayer_19, WavePlayer_20], properties {weights [1,1]}}
 {id=Random_6, children [WavePlayer_21, WavePlayer_22, WavePlayer_23], properties {weights [1,1,1]}}
 {id=Random_7, children [WavePlayer_25, WavePlayer_18], properties {bRandomizeWithoutReplacement=false, weights [1,1]}}
 {id=WavePlayer_13, properties {soundWaveAssetPtr=/Clyde/Sounds/Fort_Impact_Sounds/Demolition/utility_pole_impact_01}}
 {id=WavePlayer_14, properties {soundWaveAssetPtr=/Clyde/Sounds/Fort_Impact_Sounds/Demolition/utility_pole_impact_02}}
 {id=WavePlayer_15, properties {soundWaveAssetPtr=/Clyde/Sounds/Fort_Impact_Sounds/Demolition/Metal_Big_03}}
 {id=WavePlayer_16, properties {soundWaveAssetPtr=/Clyde/Sounds/Fort_Impact_Sounds/Demolition/Metal_Big_04}}
 {id=WavePlayer_18, properties {soundWaveAssetPtr=/Clyde/Sounds/Fort_Impact_Sounds/Demolition/Metal_Pole_Death06}}
 {id=WavePlayer_19, properties {soundWaveAssetPtr=&utility_pole_impact_01}}
 {id=WavePlayer_20, properties {soundWaveAssetPtr=&utility_pole_impact_02}}
 {id=WavePlayer_21, properties {soundWaveAssetPtr=&Metal_Big_03}}
 {id=WavePlayer_22, properties {soundWaveAssetPtr=&Metal_Big_04}}
 {id=WavePlayer_23, properties {soundWaveAssetPtr=/Clyde/Sounds/Fort_Impact_Sounds/Demolition/Metal_Big_06}}
 {id=WavePlayer_25, properties {soundWaveAssetPtr=/Clyde/Sounds/Fort_Impact_Sounds/Demolition/Metal_Pole_Death03}}
 {id=WavePlayer_3, properties {soundWaveAssetPtr=/Clyde/Sounds/Fort_Impact_Sounds/Demolition/MetalUtilityPole_ImpactFinal_opt}}
 {id=WavePlayer_38, properties {soundWaveAssetPtr=&Metal_Pole_Death06}}
 {id=WavePlayer_39, properties {soundWaveAssetPtr=/Clyde/Sounds/Fort_Impact_Sounds/Demolition/Metal_Pole_Death02}}
 {id=WavePlayer_4, properties {soundWaveAssetPtr=&MetalUtilityPole_ImpactFinal_opt}}
 {id=WavePlayer_40, properties {soundWaveAssetPtr=&Metal_Pole_Death03}}
]
properties {
 maxDistance=2097152
 soundClassObject=&Fort_Impacts_PSM
 volumeMultiplier=0.34
 concurrencyOverrides {resolutionRule=PreventNew}
 concurrencySet [&CONC_Object_Impacts_FourMax]
}
