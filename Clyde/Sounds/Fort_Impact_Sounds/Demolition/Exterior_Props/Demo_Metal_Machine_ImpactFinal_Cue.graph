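# Sound cue for the final impact of demolishing a metal machine exterior prop, routed to the Fort impacts sound class.
# A quality-level switch selects one of four branches: the first two mix pitch-varied random metal impacts with a random explosion layer, using four and two metal variants respectively.
# The last two branches play only a single pitch-varied optimized wave.
# The object-impacts concurrency set (FourMax) limits simultaneous playback.
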SoundCue /Clyde/Sounds/Fort_Impact_Sounds/Demolition/Exterior_Props/Demo_Metal_Machine_ImpactFinal_Cue
root=QualityLevel_0
nodes [
 {id=Mixer_0, children [Modulator_0,Random_1], properties {inputVolume [1,1]}}
 {id=Mixer_1, children [Modulator_2,Random_21], properties {inputVolume [1,1]}}
 {id=Modulator_0, children [Random_0], properties {pitchMax=1.07, pitchMin=0.92}}
 {id=Modulator_2, children [Random_18], properties {pitchMax=1.07, pitchMin=0.92}}
 {id=Modulator_3, children [WavePlayer_10], properties {pitchMax=1.07, pitchMin=0.92}}
 {id=Modulator_4, children [WavePlayer_11], properties {pitchMax=1.07, pitchMin=0.92}}
 {id=QualityLevel_0, children [Mixer_1,Mixer_0,Modulator_3,Modulator_4]}
 {id=Random_0, children [WavePlayer_0,WavePlayer_1], properties {weights [1,1]}}
 {id=Random_1, children [WavePlayer_4,WavePlayer_5], properties {weights [1,1]}}
 {id=Random_18, children [WavePlayer_66, WavePlayer_67, WavePlayer_68, WavePlayer_69], properties {weights [1,1,1,1]}}
 {id=Random_21, children [WavePlayer_74, WavePlayer_75], properties {weights [1,1]}}
 {id=WavePlayer_0, properties {soundWaveAssetPtr=/Clyde/Sounds/Fort_Impact_Sounds/Demolition/Metal_Machine_ImpactFinal01}}
 {id=WavePlayer_1, properties {soundWaveAssetPtr=/Clyde/Sounds/Fort_Impact_Sounds/Demolition/Metal_Machine_ImpactFinal02}}
 {id=WavePlayer_10, properties {soundWaveAssetPtr=/Clyde/Sounds/Fort_Impact_Sounds/Demolition/Metal_Machine_ImpactFinal_opt}}
 {id=WavePlayer_11, properties {soundWaveAssetPtr=&Metal_Machine_ImpactFinal_opt}}
 {id=WavePlayer_4, properties {soundWaveAssetPtr=/Clyde/Sounds/Fort_Impact_Sounds/Demolition/Explo_Impact_Layer01}}
 {id=WavePlayer_5, properties {soundWaveAssetPtr=/Clyde/Sounds/Fort_Impact_Sounds/Demolition/Explo_Impact_Layer02}}
 {id=WavePlayer_66, properties {soundWaveAssetPtr=&Metal_Machine_ImpactFinal01}}
 {id=WavePlayer_67, properties {soundWaveAssetPtr=&Metal_Machine_ImpactFinal02}}
 {id=WavePlayer_68, properties {soundWaveAssetPtr=/Clyde/Sounds/Fort_Impact_Sounds/Demolition/Metal_Machine_ImpactFinal03}}
 {id=WavePlayer_69, properties {soundWaveAssetPtr=/Clyde/Sounds/Fort_Impact_Sounds/Demolition/Metal_Machine_ImpactFinal04}}
 {id=WavePlayer_74, properties {soundWaveAssetPtr=&Explo_Impact_Layer01}}
 {id=WavePlayer_75, properties {soundWaveAssetPtr=&Explo_Impact_Layer02}}
]
properties {
 maxDistance=2097152
 soundClassObject=&Fort_Impacts_PSM
 volumeMultiplier=0.408
 concurrencyOverrides {resolutionRule=PreventNew}
 concurrencySet [&CONC_Object_Impacts_FourMax]
}
