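# Sound cue for demolition impacts on mushroom exterior props, in the Fort_Impacts sound class.
# A random node picks one of two equally weighted impact waves and feeds an enveloper whose volume curve fades out over about a second and a half.
# A root modulator adds slight random pitch variation.
# The cue plays at reduced volume and references a four-max object-impact concurrency set to limit stacking.
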SoundCue /Clyde/Sounds/Fort_Impact_Sounds/Demolition/Exterior_Props/Demo_Mushroom_Impact_Cue
root=Modulator_2
nodes [
 {
  id=Enveloper_0
  children [Random_12]
  properties {pitchCurve {editorCurveData {keys [{time=0}]}}, volumeCurve {editorCurveData {keys [<at 0 1 RCIM_Cubic>,<at 1.5429 -0.0334>]}}}
 }
 {id=Modulator_2, children [Enveloper_0], properties {pitchMax=1.07, pitchMin=0.92}}
 {id=Random_12, children [WavePlayer_0,WavePlayer_1], properties {weights [1,1]}}
 {id=WavePlayer_0, properties {soundWaveAssetPtr=/Clyde/Sounds/Fort_Impact_Sounds/Demolition/Demo_Mushroom_Impact_02}}
 {id=WavePlayer_1, properties {soundWaveAssetPtr=/Clyde/Sounds/Fort_Impact_Sounds/Demolition/Demo_Mushroom_Impact_01}}
]
properties {
 maxDistance=2097152
 soundClassObject=&Fort_Impacts
 volumeMultiplier=0.68
 concurrencyOverrides {maxCount=4}
 concurrencySet [&CONC_Object_Impacts_FourMax]
}
