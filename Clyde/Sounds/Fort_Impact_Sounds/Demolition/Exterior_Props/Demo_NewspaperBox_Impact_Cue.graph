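# Sound cue for newspaper box impacts, filed under demolition exterior props.
# A Random node picks one of three equally weighted impact waves and feeds a Modulator that slightly varies pitch on each play.
# The cue plays at reduced volume in the Fort_Impacts sound class and references a shared object-impact concurrency asset named for a four-voice limit.
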SoundCue /Clyde/Sounds/Fort_Impact_Sounds/Demolition/Exterior_Props/Demo_NewspaperBox_Impact_Cue
root=Modulator_10
nodes [
 {id=Modulator_10, children [Random_25], properties {pitchMax=1.07, pitchMin=0.92}}
 {id=Random_25, children [WavePlayer_90, WavePlayer_91, WavePlayer_92], properties {weights [1,1,1]}}
 {id=WavePlayer_90, properties {soundWaveAssetPtr=/Clyde/Sounds/Fort_Impact_Sounds/Demolition/NewspaperBox_Impact01}}
 {id=WavePlayer_91, properties {soundWaveAssetPtr=/Clyde/Sounds/Fort_Impact_Sounds/Demolition/NewspaperBox_Impact02}}
 {id=WavePlayer_92, properties {soundWaveAssetPtr=/Clyde/Sounds/Fort_Impact_Sounds/Demolition/NewspaperBox_Impact03}}
]
properties {
 maxDistance=2097152
 soundClassObject=&Fort_Impacts
 volumeMultiplier=0.68
 concurrencyOverrides {maxCount=4}
 concurrencySet [&CONC_Object_Impacts_FourMax]
}
