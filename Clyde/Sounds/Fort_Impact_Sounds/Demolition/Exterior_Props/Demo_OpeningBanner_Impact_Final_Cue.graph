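# Sound cue for the final impact of a demolished opening banner, an exterior prop in the Fort impact sound set.
# A mixer plays two wave players at once, layering a banner final impact sound with a book impact sound.
# The cue routes to the Fort impacts sound class at reduced volume and references a four-max object impact concurrency set, with a prevent-new override value.
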SoundCue /Clyde/Sounds/Fort_Impact_Sounds/Demolition/Exterior_Props/Demo_OpeningBanner_Impact_Final_Cue
root=Mixer_1
nodes [
 {id=Mixer_1, children [WavePlayer_3,WavePlayer_4], properties {inputVolume [1,1]}}
 {id=WavePlayer_3, properties {soundWaveAssetPtr=&/Demo_Banner_Impact_Final_01}}
 {id=WavePlayer_4, properties {soundWaveAssetPtr=/Clyde/Sounds/Fort_Impact_Sounds/Demolition/Book_Impact03}}
]
properties {
 maxDistance=2097152
 soundClassObject=&Fort_Impacts_PSM
 volumeMultiplier=0.68
 concurrencyOverrides {resolutionRule=PreventNew}
 concurrencySet [&CONC_Object_Impacts_FourMax]
}
