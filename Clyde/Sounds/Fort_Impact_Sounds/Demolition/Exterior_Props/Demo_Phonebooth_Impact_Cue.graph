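# Sound cue for impacts on a demolition exterior phone booth prop.
# A single phone booth medium impact wave plays through a modulator that randomizes pitch around normal for variation.
# It routes to the Fort impacts sound class at reduced volume, with a concurrency set limiting simultaneous object impacts to four.
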SoundCue /Clyde/Sounds/Fort_Impact_Sounds/Demolition/Exterior_Props/Demo_Phonebooth_Impact_Cue
root=Modulator_14
nodes [
 {id=Modulator_14, children [WavePlayer_3], properties {pitchMax=1.2, pitchMin=0.8}}
 {id=WavePlayer_3, properties {soundWaveAssetPtr=&/phonebooth_med_impact_01}}
]
properties {
 maxDistance=2097152
 soundClassObject=&Fort_Impacts
 volumeMultiplier=0.68
 concurrencyOverrides {maxCount=4}
 concurrencySet [&CONC_Object_Impacts_FourMax]
}
