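# Sound cue for small plastic prop impacts, filed under exterior demolition sounds and routed to the Fort_Impacts sound class.
# A quality-level root selects one of four branches, whose wave variety shrinks from three random variants through two random pairs to a single wave.
# Each branch applies the same slight random pitch modulation to avoid obvious repetition.
# The shared CONC_Object_Impacts_FourMax concurrency set governs simultaneous impacts.
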SoundCue /Clyde/Sounds/Fort_Impact_Sounds/Demolition/Exterior_Props/Demo_Plastic_Small_Impact_Cue
root=QualityLevel_0
nodes [
 {id=Modulator_2, children [Random_4], properties {pitchMax=1.07, pitchMin=0.92}}
 {id=Modulator_3, children [Random_5], properties {pitchMax=1.07, pitchMin=0.92}}
 {id=Modulator_4, children [Random_6], properties {pitchMax=1.07, pitchMin=0.92}}
 {id=Modulator_5, children [WavePlayer_17], properties {pitchMax=1.07, pitchMin=0.92}}
 {id=QualityLevel_0, children [Modulator_2,Modulator_3,Modulator_4,Modulator_5]}
 {id=Random_4, children [WavePlayer_10, WavePlayer_11, WavePlayer_12], properties {weights [1,1,1]}}
 {id=Random_5, children [WavePlayer_13, WavePlayer_14], properties {bRandomizeWithoutReplacement=false, weights [1,1]}}
 {id=Random_6, children [WavePlayer_15, WavePlayer_16], properties {bRandomizeWithoutReplacement=false, weights [1,1]}}
 {id=WavePlayer_10, properties {soundWaveAssetPtr=/Clyde/Sounds/Fort_Impact_Sounds/Demolition/Plastic_Small_Impact_mix_01}}
 {id=WavePlayer_11, properties {soundWaveAssetPtr=/Clyde/Sounds/Fort_Impact_Sounds/Demolition/Plastic_Small_Impact_mix_02}}
 {id=WavePlayer_12, properties {soundWaveAssetPtr=/Clyde/Sounds/Fort_Impact_Sounds/Demolition/Plastic_Small_Impact_mix_03}}
 {id=WavePlayer_13, properties {soundWaveAssetPtr=&Plastic_Small_Impact_mix_01}}
 {id=WavePlayer_14, properties {soundWaveAssetPtr=&Plastic_Small_Impact_mix_02}}
 {id=WavePlayer_15, properties {soundWaveAssetPtr=&Plastic_Small_Impact_mix_01}}
 {id=WavePlayer_16, properties {soundWaveAssetPtr=&Plastic_Small_Impact_mix_02}}
 {id=WavePlayer_17, properties {soundWaveAssetPtr=&Plastic_Small_Impact_mix_01}}
]
properties {
 maxDistance=2097152
 soundClassObject=&Fort_Impacts
 volumeMultiplier=0.68
 concurrencyOverrides {maxCount=3, resolutionRule=StopOldest}
 concurrencySet [&CONC_Object_Impacts_FourMax]
}
