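# Sound cue for demolition impacts on porta-potty exterior props, routed to the Fort impacts sound class with a four-max object impact concurrency.
# A mixer layers two branches: a random car medium impact at full level and a random bus medium impact at half level.
# Each branch passes through a modulator and its own attenuation override, with the bus layer using custom falloff and shape settings.
# The cue is slightly pitched up and reduced in volume.
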SoundCue /Clyde/Sounds/Fort_Impact_Sounds/Demolition/Exterior_Props/Demo_PortaPotty_Impact_Cue
root=Mixer_3
nodes [
 {
  id=Attenuation_4
  children [Modulator_0]
  properties {bOverrideAttenuation=true, attenuationOverrides {falloffDistance=900, attenuationShapeExtents {x=100}}}
 }
 {id=Attenuation_5, children [Modulator_1], properties {bOverrideAttenuation=true}}
 {id=Mixer_3, children [Attenuation_5, Attenuation_4], properties {inputVolume [1,0.5]}}
 {id=Modulator_0, children [Random_2]}
 {id=Modulator_1, children [Random_6]}
 {id=Random_2, children [WavePlayer_14, WavePlayer_15, WavePlayer_16], properties {weights [1,1,1]}}
 {id=Random_6, children [WavePlayer_23, WavePlayer_24, WavePlayer_25], properties {weights [1,1,1]}}
 {id=WavePlayer_14, properties {soundWaveAssetPtr=/Clyde/Sounds/Fort_Impact_Sounds/Demolition/Bus_ImpactMedium01}}
 {id=WavePlayer_15, properties {soundWaveAssetPtr=/Clyde/Sounds/Fort_Impact_Sounds/Demolition/Bus_ImpactMedium02}}
 {id=WavePlayer_16, properties {soundWaveAssetPtr=/Clyde/Sounds/Fort_Impact_Sounds/Demolition/Bus_ImpactMedium04}}
 {id=WavePlayer_23, properties {soundWaveAssetPtr=/Clyde/Sounds/Fort_Impact_Sounds/Demolition/Car_Impact_Med2}}
 {id=WavePlayer_24, properties {soundWaveAssetPtr=/Clyde/Sounds/Fort_Impact_Sounds/Demolition/Car_Impact_Med3}}
 {id=WavePlayer_25, properties {soundWaveAssetPtr=/Clyde/Sounds/Fort_Impact_Sounds/Demolition/Car_Impact_Med4}}
]
properties {
 maxDistance=4000
 pitchMultiplier=1.1
 soundClassObject=&Fort_Impacts
 volumeMultiplier=0.68
 concurrencyOverrides {maxCount=4}
 concurrencySet [&CONC_Object_Impacts_FourMax]
}
