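# Sound cue for demolition of the PvP Cornerstone exterior prop, playing rock impacts in the Fort_Impacts sound class.
# A distance crossfade blends a close layer of RockFinal waves into a farther layer of rock_newimp waves that carries out to the cue's max distance.
# Each layer randomly picks one of three waves, then applies a fixed-volume modulator and its own attenuation override.
# A four-max object-impact concurrency set limits simultaneous plays.
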SoundCue /Clyde/Sounds/Fort_Impact_Sounds/Demolition/Exterior_Props/Demo_PvP_Cornerstone_Cue
root=DistanceCrossFade_2
nodes [
 {
  id=Attenuation_4
  children [Modulator_0]
  properties {bOverrideAttenuation=true, attenuationOverrides {falloffDistance=900, attenuationShapeExtents {x=100}}}
 }
 {id=Attenuation_5, children [Modulator_1], properties {bOverrideAttenuation=true}}
 {
  id=DistanceCrossFade_2
  children [Attenuation_4, Attenuation_5]
  properties {crossFadeInput [
    {fadeInDistanceEnd=0, fadeInDistanceStart=0, fadeOutDistanceEnd=800, fadeOutDistanceStart=400, volume=1}
    {fadeInDistanceEnd=400, fadeInDistanceStart=200, fadeOutDistanceEnd=7000, fadeOutDistanceStart=5000, volume=0.7}
   ]}
 }
 {id=Modulator_0, children [Random_0], properties {volumeMax=0.5, volumeMin=0.5}}
 {id=Modulator_1, children [Random_1], properties {volumeMax=2, volumeMin=2}}
 {id=Random_0, children [WavePlayer_1,WavePlayer_2,WavePlayer_3], properties {weights [1,1,1]}}
 {id=Random_1, children [WavePlayer_4,WavePlayer_5,WavePlayer_7], properties {weights [1,1,1]}}
 {id=WavePlayer_1, properties {soundWaveAssetPtr=/Clyde/Sounds/Fort_Impact_Sounds/Demolition/Fort_Impact_RockFinal_01}}
 {id=WavePlayer_2, properties {soundWaveAssetPtr=/Clyde/Sounds/Fort_Impact_Sounds/Demolition/Fort_Impact_RockFinal_02}}
 {id=WavePlayer_3, properties {soundWaveAssetPtr=/Clyde/Sounds/Fort_Impact_Sounds/Demolition/Fort_Impact_RockFinal_03}}
 {id=WavePlayer_4, properties {soundWaveAssetPtr=/Clyde/Sounds/Fort_Impact_Sounds/Demolition/rock_newimp_10}}
 {id=WavePlayer_5, properties {soundWaveAssetPtr=/Clyde/Sounds/Fort_Impact_Sounds/Demolition/rock_newimp_11}}
 {id=WavePlayer_7, properties {soundWaveAssetPtr=/Clyde/Sounds/Fort_Impact_Sounds/Demolition/rock_newimp_01}}
]
properties {
 maxDistance=7000
 soundClassObject=&Fort_Impacts
 volumeMultiplier=0.68
 concurrencyOverrides {maxCount=4}
 concurrencySet [&CONC_Object_Impacts_FourMax]
}
