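# Sound cue for rubber impact sounds, filed under demolition exterior props in the impact sound set.
# A Random node picks one of three equally weighted rubber impact waves, and a Modulator slightly varies its pitch.
# The cue plays at reduced volume in the Fort_Impacts sound class and references a shared object-impact concurrency set to limit simultaneous plays.
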SoundCue /Clyde/Sounds/Fort_Impact_Sounds/Demolition/Exterior_Props/Demo_Rubber_Impact_Cue
root=Modulator_2
nodes [
 {id=Modulator_2, children [Random_31], properties {pitchMax=1.07, pitchMin=0.92}}
 {id=Random_31, children [WavePlayer_91, WavePlayer_92, WavePlayer_93], properties {weights [1,1,1]}}
 {id=WavePlayer_91, properties {soundWaveAssetPtr=/Clyde/Sounds/Fort_Impact_Sounds/Demolition/Rubber_Impact01}}
 {id=WavePlayer_92, properties {soundWaveAssetPtr=/Clyde/Sounds/Fort_Impact_Sounds/Demolition/Rubber_Impact02}}
 {id=WavePlayer_93, properties {soundWaveAssetPtr=/Clyde/Sounds/Fort_Impact_Sounds/Demolition/Rubber_Impact03}}
]
properties {
 maxDistance=2097152
 soundClassObject=&Fort_Impacts
 volumeMultiplier=0.68
 concurrencyOverrides {maxCount=3, resolutionRule=StopOldest}
 concurrencySet [&CONC_Object_Impacts_FourMax]
}
