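# Sound cue for sandcastle prop impacts, part of the demolition exterior prop impact sounds.
# A Random node picks one of four equally weighted impact waves and feeds a root Modulator that varies pitch, with a raised upper limit.
# It uses the Fort_Impacts sound class and a shared object-impact concurrency set to cap simultaneous plays.
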SoundCue /Clyde/Sounds/Fort_Impact_Sounds/Demolition/Exterior_Props/Demo_SandCastle_Impact_Cue
root=Modulator_5
nodes [
 {id=Modulator_5, children [Random_4], properties {pitchMax=1.3}}
 {id=Random_4, children [WavePlayer_5, WavePlayer_12, WavePlayer_13, WavePlayer_14], properties {weights [1,1,1,1]}}
 {id=WavePlayer_12, properties {soundWaveAssetPtr=&/Demo_SandCastle_Impact_02}}
 {id=WavePlayer_13, properties {soundWaveAssetPtr=&/Demo_SandCastle_Impact_03}}
 {id=WavePlayer_14, properties {soundWaveAssetPtr=&/Demo_SandCastle_Impact_04}}
 {id=WavePlayer_5, properties {soundWaveAssetPtr=&/Demo_SandCastle_Impact_01}}
]
properties {
 maxDistance=2097152
 soundClassObject=&Fort_Impacts
 volumeMultiplier=0.96
 concurrencyOverrides {resolutionRule=PreventNew}
 concurrencySet [&CONC_Object_Impacts_FourMax]
}
