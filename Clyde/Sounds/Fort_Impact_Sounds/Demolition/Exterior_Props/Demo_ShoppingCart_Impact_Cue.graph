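# Sound cue for shopping cart impacts in demolition, assigned to the Fort_Impacts sound class.
# A root quality-level node selects among four branches: the first two layer random metal machine hits with higher-pitched random cart impacts, the first having one extra cart variant.
# The last two each play a single optimized cart wave.
# All branches randomize pitch, and concurrency uses a four-max object impact set plus a PreventNew override.
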SoundCue /Clyde/Sounds/Fort_Impact_Sounds/Demolition/Exterior_Props/Demo_ShoppingCart_Impact_Cue
root=QualityLevel_0
nodes [
 {id=Mixer_0, children [Modulator_2,Modulator_0], properties {inputVolume [1,1]}}
 {id=Mixer_4, children [Modulator_1,Modulator_7], properties {inputVolume [1,1]}}
 {id=Modulator_0, children [Random_0], properties {pitchMax=1.5, pitchMin=1.2}}
 {id=Modulator_1, children [Random_2], properties {pitchMax=1.2, pitchMin=0.92}}
 {id=Modulator_2, children [Random_3], properties {pitchMax=1.2, pitchMin=0.92}}
 {id=Modulator_3, children [WavePlayer_10], properties {pitchMax=1.2, pitchMin=0.92}}
 {id=Modulator_4, children [WavePlayer_11], properties {pitchMax=1.2, pitchMin=0.92}}
 {id=Modulator_7, children [Random_1], properties {pitchMax=1.5, pitchMin=1.2}}
 {id=QualityLevel_0, children [Mixer_4,Mixer_0,Modulator_3,Modulator_4]}
 {id=Random_0, children [WavePlayer_0,WavePlayer_1], properties {weights [1,1]}}
 {id=Random_1, children [WavePlayer_2,WavePlayer_3,WavePlayer_4], properties {weights [1,1,1]}}
 {id=Random_2, children [WavePlayer_8,WavePlayer_9], properties {weights [1,1]}}
 {id=Random_3, children [WavePlayer_6,WavePlayer_7], properties {weights [1,1]}}
 {id=WavePlayer_0, properties {soundWaveAssetPtr=/Clyde/Sounds/Fort_Impact_Sounds/Demolition/shoppingcart_impact_01}}
 {id=WavePlayer_1, properties {soundWaveAssetPtr=/Clyde/Sounds/Fort_Impact_Sounds/Demolition/shoppingcart_impact_02}}
 {id=WavePlayer_10, properties {soundWaveAssetPtr=/Clyde/Sounds/Fort_Impact_Sounds/Demolition/ShoppingCart_Impact02_opt}}
 {id=WavePlayer_11, properties {soundWaveAssetPtr=&ShoppingCart_Impact02_opt}}
 {id=WavePlayer_2, properties {soundWaveAssetPtr=&shoppingcart_impact_01}}
 {id=WavePlayer_3, properties {soundWaveAssetPtr=&shoppingcart_impact_02}}
 {id=WavePlayer_4, properties {soundWaveAssetPtr=/Clyde/Sounds/Fort_Impact_Sounds/Demolition/shoppingcart_impact_03}}
 {id=WavePlayer_6, properties {soundWaveAssetPtr=/Clyde/Sounds/Fort_Impact_Sounds/Demolition/Metal_Machine_Impact02}}
 {id=WavePlayer_7, properties {soundWaveAssetPtr=/Clyde/Sounds/Fort_Impact_Sounds/Demolition/Metal_Machine_Impact03}}
 {id=WavePlayer_8, properties {soundWaveAssetPtr=&Metal_Machine_Impact02}}
 {id=WavePlayer_9, properties {soundWaveAssetPtr=&Metal_Machine_Impact03}}
]
properties {
 maxDistance=2097152
 soundClassObject=&Fort_Impacts
 volumeMultiplier=0.68
 concurrencyOverrides {resolutionRule=PreventNew}
 concurrencySet [&CONC_Object_Impacts_FourMax]
}
