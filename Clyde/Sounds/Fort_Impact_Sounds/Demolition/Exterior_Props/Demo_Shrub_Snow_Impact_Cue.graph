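# Sound cue for snowy shrub impacts in demolition, routed to the Fort_Impacts sound class with a four-max concurrency set that prevents new sounds at the limit.
# A quality-level root picks among four branches: two mix pitch-varied random shrub and snow impact layers, the second using fewer variations.
# The other two each play one optimized pine-tree snow impact, pitched up and quieter, as a cheaper substitute for the layered mix.
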SoundCue /Clyde/Sounds/Fort_Impact_Sounds/Demolition/Exterior_Props/Demo_Shrub_Snow_Impact_Cue
root=QualityLevel_0
nodes [
 {id=Mixer_0, children [Modulator_2,Modulator_0], properties {inputVolume [1,1]}}
 {id=Mixer_1, children [Modulator_7,Modulator_6], properties {inputVolume [1,1]}}
 {id=Modulator_0, children [Random_0], properties {pitchMax=1.07, pitchMin=0.92}}
 {id=Modulator_2, children [Random_1], properties {pitchMax=1.07, pitchMin=0.92}}
 {id=Modulator_3, children [WavePlayer_9], properties {pitchMax=1.25, pitchMin=1.1, volumeMax=0.5, volumeMin=0.45}}
 {id=Modulator_5, children [WavePlayer_16], properties {pitchMax=1.25, pitchMin=1.1, volumeMax=0.5, volumeMin=0.45}}
 {id=Modulator_6, children [Random_4], properties {pitchMax=1.07, pitchMin=0.92}}
 {id=Modulator_7, children [Random_2], properties {pitchMax=1.07, pitchMin=0.92}}
 {id=QualityLevel_0, children [Mixer_0,Mixer_1,Modulator_3,Modulator_5]}
 {id=Random_0, children [WavePlayer_0,WavePlayer_1,WavePlayer_2,WavePlayer_3,WavePlayer_4,WavePlayer_5], properties {weights [1,1,1,1,1,1]}}
 {id=Random_1, children [WavePlayer_6,WavePlayer_7,WavePlayer_8], properties {weights [1,1,1]}}
 {id=Random_2, children [WavePlayer_12, WavePlayer_17], properties {bRandomizeWithoutReplacement=false, weights [1,1]}}
 {id=Random_4, children [WavePlayer_10, WavePlayer_11], properties {weights [1,1]}}
 {id=WavePlayer_0, properties {soundWaveAssetPtr=/Clyde/Sounds/Fort_Impact_Sounds/Demolition/Wood/Trees_Holiday_Winter_Snow_Impacts_2017-001}}
 {id=WavePlayer_1, properties {soundWaveAssetPtr=/Clyde/Sounds/Fort_Impact_Sounds/Demolition/Wood/Trees_Holiday_Winter_Snow_Impacts_2017-002}}
 {id=WavePlayer_10, properties {soundWaveAssetPtr=&Trees_Holiday_Winter_Snow_Impacts_2017-001}}
 {id=WavePlayer_11, properties {soundWaveAssetPtr=&Trees_Holiday_Winter_Snow_Impacts_2017-002}}
 {id=WavePlayer_12, properties {soundWaveAssetPtr=/Clyde/Sounds/Fort_Impact_Sounds/Demolition/shrub_impact_01}}
 {id=WavePlayer_16, properties {soundWaveAssetPtr=/Clyde/Sounds/Fort_Impact_Sounds/Demolition/Wood/PineTree_Impact_Snow_opt}}
 {id=WavePlayer_17, properties {soundWaveAssetPtr=/Clyde/Sounds/Fort_Impact_Sounds/Demolition/shrub_impact_02}}
 {id=WavePlayer_2, properties {soundWaveAssetPtr=/Clyde/Sounds/Fort_Impact_Sounds/Demolition/Wood/Trees_Holiday_Winter_Snow_Impacts_2017-003}}
 {id=WavePlayer_3, properties {soundWaveAssetPtr=/Clyde/Sounds/Fort_Impact_Sounds/Demolition/Wood/Trees_Holiday_Winter_Snow_Impacts_2017-004}}
 {id=WavePlayer_4, properties {soundWaveAssetPtr=/Clyde/Sounds/Fort_Impact_Sounds/Demolition/Wood/Trees_Holiday_Winter_Snow_Impacts_2017-005}}
 {id=WavePlayer_5, properties {soundWaveAssetPtr=/Clyde/Sounds/Fort_Impact_Sounds/Demolition/Wood/Trees_Holiday_Winter_Snow_Impacts_2017-006}}
 {id=WavePlayer_6, properties {soundWaveAssetPtr=&shrub_impact_01}}
 {id=WavePlayer_7, properties {soundWaveAssetPtr=&shrub_impact_02}}
 {id=WavePlayer_8, properties {soundWaveAssetPtr=/Clyde/Sounds/Fort_Impact_Sounds/Demolition/shrub_impact_03}}
 {id=WavePlayer_9, properties {soundWaveAssetPtr=&PineTree_Impact_Snow_opt}}
]
properties {
 maxDistance=2097152
 soundClassObject=&Fort_Impacts
 volumeMultiplier=0.68
 concurrencyOverrides {resolutionRule=PreventNew}
 concurrencySet [&CONC_Object_Impacts_FourMax]
}
