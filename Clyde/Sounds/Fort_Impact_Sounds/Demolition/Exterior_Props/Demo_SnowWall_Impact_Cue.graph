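# Sound cue for a demolition snow wall prop impact, built by layering existing snow-themed impact sounds.
# A mixer combines a slightly quieter, immediate Yeti pickaxe hit with short fixed-delay snowball hits and, slightly later, small Sneaky Snowman impacts.
# Each layer picks randomly among equally weighted variations, and a root modulator randomizes pitch over the whole mix.
# It plays in the impacts sound class with a four-max object impact concurrency set.
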SoundCue /Clyde/Sounds/Fort_Impact_Sounds/Demolition/Exterior_Props/Demo_SnowWall_Impact_Cue
root=Modulator_2
nodes [
 {id=Delay_0, children [Random_5], properties {delayMax=0.08, delayMin=0.08}}
 {id=Delay_1, children [Random_4], properties {delayMax=0.05, delayMin=0.05}}
 {id=Mixer_0, children [Delay_0,Random_2,Delay_1], properties {inputVolume [1,0.7,1]}}
 {id=Modulator_2, children [Mixer_0], properties {pitchMax=1.07, pitchMin=0.92}}
 {id=Random_2, children [WavePlayer_0,WavePlayer_6,WavePlayer_8], properties {weights [1,1,1]}}
 {id=Random_4, children [WavePlayer_15, WavePlayer_16, WavePlayer_17, WavePlayer_18], properties {weights [1,1,1,1]}}
 {id=Random_5, children [WavePlayer_19, WavePlayer_20, WavePlayer_21], properties {weights [1,1,1]}}
 {id=WavePlayer_0, properties {soundWaveAssetPtr=/Game/Athena/Sounds/Weapons/PickAxes/Yeti/PA_Yeti_Impact_02}}
 {id=WavePlayer_15, properties {soundWaveAssetPtr=/Game/Sounds/Fort_Weapons/Impacts/IMPCT_SnowBall_Hit_1}}
 {id=WavePlayer_16, properties {soundWaveAssetPtr=/Game/Sounds/Fort_Weapons/Impacts/IMPCT_SnowBall_Hit_2}}
 {id=WavePlayer_17, properties {soundWaveAssetPtr=/Game/Sounds/Fort_Weapons/Impacts/IMPCT_SnowBall_Hit_3}}
 {id=WavePlayer_18, properties {soundWaveAssetPtr=/Game/Sounds/Fort_Weapons/Impacts/IMPCT_SnowBall_Hit_4}}
 {id=WavePlayer_19, properties {soundWaveAssetPtr=/Game/Sounds/Fort_Gadgets/SneakySnowman/Version2/SneakySnowman_Impact_Small_01}}
 {id=WavePlayer_20, properties {soundWaveAssetPtr=/Game/Sounds/Fort_Gadgets/SneakySnowman/Version2/SneakySnowman_Impact_Small_02}}
 {id=WavePlayer_21, properties {soundWaveAssetPtr=/Game/Sounds/Fort_Gadgets/SneakySnowman/Version2/SneakySnowman_Impact_Small_03}}
 {id=WavePlayer_6, properties {soundWaveAssetPtr=/Game/Athena/Sounds/Weapons/PickAxes/Yeti/PA_Yeti_Impact_03}}
 {id=WavePlayer_8, properties {soundWaveAssetPtr=/Game/Athena/Sounds/Weapons/PickAxes/Yeti/PA_Yeti_Impact_04}}
]
properties {
 maxDistance=2097152
 soundClassObject=&Fort_Impacts
 volumeMultiplier=0.68
 concurrencyOverrides {maxCount=4}
 concurrencySet [&CONC_Object_Impacts_FourMax]
}
