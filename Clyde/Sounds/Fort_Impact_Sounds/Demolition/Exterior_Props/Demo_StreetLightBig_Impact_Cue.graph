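# Sound cue for the impact of a large street light prop, routed to the impacts sound class with four-max concurrency.
# A quality-level switch selects among four branches that get simpler down the list.
# The top two mix a random, pitched-down streetlight impact with a slightly delayed, pitched-up large pipe hit.
# The third keeps only the random streetlight layer, and the last plays a single pitched-down wave.
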SoundCue /Clyde/Sounds/Fort_Impact_Sounds/Demolition/Exterior_Props/Demo_StreetLightBig_Impact_Cue
root=QualityLevel_0
nodes [
 {id=Delay_0, children [WavePlayer_8], properties {delayMax=0.04, delayMin=0.02}}
 {id=Delay_1, children [WavePlayer_2], properties {delayMax=0.04, delayMin=0.02}}
 {id=Mixer_0, children [Modulator_0,Modulator_1], properties {inputVolume [1,1]}}
 {id=Mixer_1, children [Modulator_2,Modulator_3], properties {inputVolume [1,1]}}
 {id=Modulator_0, children [Random_0], properties {pitchMax=0.7, pitchMin=0.6}}
 {id=Modulator_1, children [Delay_1], properties {pitchMax=1.6, pitchMin=1.5}}
 {id=Modulator_2, children [Random_2], properties {pitchMax=0.7, pitchMin=0.6}}
 {id=Modulator_3, children [Delay_0], properties {pitchMax=1.6, pitchMin=1.5}}
 {id=Modulator_4, children [Mixer_0], properties {volumeMax=0.7, volumeMin=0.7}}
 {id=Modulator_5, children [Random_1], properties {pitchMax=0.7, pitchMin=0.6}}
 {id=Modulator_6, children [Modulator_5], properties {volumeMax=1, volumeMin=0.9}}
 {id=Modulator_7, children [Mixer_1], properties {volumeMax=0.7, volumeMin=0.7}}
 {id=Modulator_8, children [Modulator_9], properties {volumeMax=1, volumeMin=0.9}}
 {id=Modulator_9, children [WavePlayer_5], properties {pitchMax=0.7, pitchMin=0.6}}
 {id=QualityLevel_0, children [Modulator_7,Modulator_4,Modulator_6,Modulator_8]}
 {id=Random_0, children [WavePlayer_0,WavePlayer_1], properties {weights [1,1]}}
 {id=Random_1, children [WavePlayer_3,WavePlayer_4], properties {weights [1,1]}}
 {id=Random_2, children [WavePlayer_6,WavePlayer_7], properties {weights [1,1]}}
 {id=WavePlayer_0, properties {soundWaveAssetPtr=&/streetlight_impact_02}}
 {id=WavePlayer_1, properties {soundWaveAssetPtr=&/streetlight_impact_01}}
 {id=WavePlayer_2, properties {soundWaveAssetPtr=&/pipe_large_impact_03}}
 {id=WavePlayer_3, properties {soundWaveAssetPtr=&streetlight_impact_02}}
 {id=WavePlayer_4, properties {soundWaveAssetPtr=&streetlight_impact_01}}
 {id=WavePlayer_5, properties {soundWaveAssetPtr=&streetlight_impact_02}}
 {id=WavePlayer_6, properties {soundWaveAssetPtr=&streetlight_impact_02}}
 {id=WavePlayer_7, properties {soundWaveAssetPtr=&streetlight_impact_01}}
 {id=WavePlayer_8, properties {soundWaveAssetPtr=&pipe_large_impact_03}}
]
properties {
 maxDistance=2097152
 soundClassObject=&Fort_Impacts
 volumeMultiplier=0.34
 concurrencyOverrides {maxCount=4}
 concurrencySet [&CONC_Object_Impacts_FourMax]
}
