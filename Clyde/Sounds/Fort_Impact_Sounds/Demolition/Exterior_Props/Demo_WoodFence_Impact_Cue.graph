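# Sound cue for wood fence impacts, filed under exterior prop demolition sounds and assigned to the impacts sound class.
# A quality-level root selects one of four branches, stepping the variation down from four random waves to three, two, then a single wave.
# Each branch runs through a modulator, with explicit pitch variation on the single-wave branch.
# Playback uses small melee-weapon attenuation and a four-voice object-impact concurrency limit.
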SoundCue /Clyde/Sounds/Fort_Impact_Sounds/Demolition/Exterior_Props/Demo_WoodFence_Impact_Cue
root=QualityLevel_0
nodes [
 {id=Modulator_0, children [Random_1]}
 {id=Modulator_1, children [Random_0]}
 {id=Modulator_2, children [Random_2]}
 {id=Modulator_3, children [WavePlayer_9], properties {pitchMax=1.1, pitchMin=0.9}}
 {id=QualityLevel_0, children [Modulator_0,Modulator_1,Modulator_2,Modulator_3]}
 {id=Random_0, children [WavePlayer_0,WavePlayer_1,WavePlayer_2], properties {weights [1,1,1]}}
 {id=Random_1, children [WavePlayer_4,WavePlayer_5,WavePlayer_6,WavePlayer_7], properties {weights [1,1,1,1]}}
 {id=Random_2, children [WavePlayer_3,WavePlayer_8], properties {weights [1,1]}}
 {id=WavePlayer_0, properties {soundWaveAssetPtr=&/fence_impact_01}}
 {id=WavePlayer_1, properties {soundWaveAssetPtr=&/fence_impact_02}}
 {id=WavePlayer_2, properties {soundWaveAssetPtr=&/fence_impact_03}}
 {id=WavePlayer_3, properties {soundWaveAssetPtr=&fence_impact_01}}
 {id=WavePlayer_4, properties {soundWaveAssetPtr=&fence_impact_01}}
 {id=WavePlayer_5, properties {soundWaveAssetPtr=&fence_impact_02}}
 {id=WavePlayer_6, properties {soundWaveAssetPtr=&fence_impact_03}}
 {id=WavePlayer_7, properties {soundWaveAssetPtr=&/fence_impact_04}}
 {id=WavePlayer_8, properties {soundWaveAssetPtr=&fence_impact_02}}
 {id=WavePlayer_9, properties {soundWaveAssetPtr=&fence_impact_01}}
]
properties {
 attenuationSettings=&Fort_Weapon_Melee_Sml_Att
 maxDistance=2800
 soundClassObject=&Fort_Impacts
 volumeMultiplier=0.504
 concurrencyOverrides {maxCount=4}
 concurrencySet [&CONC_Object_Impacts_FourMax]
}
