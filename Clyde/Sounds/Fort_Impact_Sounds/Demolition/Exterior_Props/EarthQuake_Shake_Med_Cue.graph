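# Sound cue for a medium earthquake shake, filed with the exterior prop demolition impact sounds.
# A single wave player feeds a modulator that randomizes playback, with an upper pitch limit raised to 1.3.
# The cue boosts volume and is assigned a third-person ranged weapon fire sound class rather than an impact-named one.
# It references a shared object impacts concurrency set and also carries stop-oldest override settings.
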SoundCue /Clyde/Sounds/Fort_Impact_Sounds/Demolition/Exterior_Props/EarthQuake_Shake_Med_Cue
root=Modulator_0
nodes [
 {id=Modulator_0, children [WavePlayer_0], properties {pitchMax=1.3}}
 {id=WavePlayer_0, properties {soundWaveAssetPtr=/Clyde/Sounds/Fort_Impact_Sounds/Demolition/EarthQuake_Shakes_Med}}
]
properties {
 maxDistance=2097152
 soundClassObject=&Fort_WeaponRanged_Fire_3P_PSM
 volumeMultiplier=1.25
 concurrencyOverrides {maxCount=3, resolutionRule=StopOldest}
 concurrencySet [&CONC_Object_Impacts_FourMax]
}
