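# Sound cue for Shelters TP impacts, filed under demolition exterior props and routed to the Fort_Impacts sound class.
# A Random node with equal weights picks one of three wave variants, numbered 01, 02 and 04.
# A Modulator at the root varies the pitch of each play, and the cue applies a slight volume boost.
# It references a shared object-impact concurrency set to limit simultaneous voices.
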SoundCue /Clyde/Sounds/Fort_Impact_Sounds/Demolition/Exterior_Props/Shelters_TP_Impact_Cue
root=Modulator_0
nodes [
 {id=Modulator_0, children [Random_0], properties {pitchMax=1.1}}
 {id=Random_0, children [WavePlayer_0,WavePlayer_1,WavePlayer_6], properties {weights [1,1,1]}}
 {id=WavePlayer_0, properties {soundWaveAssetPtr=&/Shelters_TP_Impact_Cue_01}}
 {id=WavePlayer_1, properties {soundWaveAssetPtr=&/Shelters_TP_Impact_Cue_02}}
 {id=WavePlayer_6, properties {soundWaveAssetPtr=&/Shelters_TP_Impact_Cue_04}}
]
properties {
 maxDistance=2097152
 soundClassObject=&Fort_Impacts
 volumeMultiplier=1.2
 concurrencyOverrides {maxCount=3, resolutionRule=StopOldest}
 concurrencySet [&CONC_Object_Impacts_FourMax]
}
